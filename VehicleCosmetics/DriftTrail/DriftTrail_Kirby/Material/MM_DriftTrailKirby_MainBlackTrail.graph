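# Masked, unlit, dithered material for the main black trail of the Kirby drift-trail vehicle cosmetic, used on Niagara ribbons and mesh particles.
# A trail base texture supplies shape masks, eroded by panning sphere-mask noise and tinted by particle color through emissive adjustments.
# Static switches select edge or body variants for color and alpha.
# A Wheel0/Ribbon1 scalar blends between center-scaled wheel UVs without foam breakup and ribbon UVs with smoothstepped, counter-panning foam gradients cutting the alpha.
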Material /VehicleCosmetics/DriftTrail/DriftTrail_Kirby/Material/MM_DriftTrailKirby_MainBlackTrail
graph {
 properties {
  bAutomaticallySetUsageInEditor=false
  bUsedWithMeshParticles=true
  bUsedWithNiagaraMeshParticles=true
  bUsedWithNiagaraRibbons=true
  blendMode=BLEND_Masked
  ditherOpacityMask=true
  shadingModel=MSM_Unlit
  usedShadingModels=Unlit
 }
 nodes [
  n1 = ParticleColor()
  n2 = TextureCoordinate()
  n3 = AppendVector(A=ScalarParameter() {defaultValue=1, parameterName=ScaleMaskWheel_U}, B=ScalarParameter() {defaultValue=1, parameterName=ScaleMaskWheel_V})
  n4 = TextureSample(UVs=LinearInterpolate(
    A=MaterialFunctionCall(UVs=@n2, Texture Scale=@n3) {materialFunction=&ScaleUVsByCenter}.UVs
    B=@n2
    Alpha=ScalarParameter() {defaultValue=1, parameterName=Wheel0/Ribbon1}
   ) {constA=1}) {
   samplerSource=SSM_Clamp_WorldGroupSettings
   texture=/VehicleCosmetics/DriftTrail/DriftTrail_Kirby/Textures/T_VCVFX_DRiftrailKirby_TrailBase_RGB
  }
  n5 = TextureCoordinate() {uTiling=7}
  n6 = Multiply(A=Time(), B=ScalarParameter() {parameterName=TimeMultiNoise})
  n7 = TextureSample(UVs=Panner(Coordinate=@n5, Time=@n6) {speedX=-2.7, speedY=0.3}) {texture=/VehicleCosmetics/DriftTrail/DriftTrail_Kirby/Textures/T_VCVFX_SphereMask_RGB}
  n8 = TextureCoordinate()
  n9 = Power(Base=Multiply(A=ComponentMask(None=@n8) {r=true}) {constB=3}) {constExponent=10}
  n10 = StaticSwitchParameter(True=Multiply(A=@n1.A, B=@n4.R), False=Subtract(A=Multiply(A=@n1.A, B=@n4.G), B=Multiply(A=@n7.G, B=@n9))) {parameterName=EdgeAlpha}
  n11 = ScalarParameter() {defaultValue=0.4, parameterName=MinSmoothStep}
  n12 = ScalarParameter() {parameterName=MaxSmoothStep}
  n13 = TextureCoordinate() {uTiling=0.7}
  n14 = Panner(Coordinate=@n13) {speedY=-0.8}
  n15 = Panner(Coordinate=@n13) {speedY=1}
  n16 = OneMinus(None=Multiply(
    A=TextureSample!(UVs=@n14, Tex, Level, Apply View MipBias) {mipValueMode=TMVM_MipLevel, texture=/VehicleCosmetics/DriftTrail/DriftTrail_Kirby/Textures/T_VCVFX_GradientKirbyFoam_A}.R
    B=TextureSample(UVs=@n15) {mipValueMode=TMVM_MipLevel, texture=&T_VCVFX_GradientKirbyFoam_A}.R
   ))
  LinearInterpolate(
   A=@n10
   B=Multiply(A=@n10, B=SmoothStep(Min=@n11, Max=@n12, Value=@n16))
   Alpha=ScalarParameter() {defaultValue=1, parameterName=Wheel0/Ribbon1}
  ) {constA=1}
  n17 = TextureSample!(UVs=Panner(Coordinate=@n5, Time=@n6) {speedX=-3, speedY=-0.5}, Tex, Apply View MipBias) {texture=&T_VCVFX_SphereMask_RGB}
  n18 = Saturate(None=Subtract(A=@n4.R, B=Add(A=@n17.R, B=@n7.G)))
  MaterialFunctionCall(Emissive=StaticSwitchParameter(True=Multiply(A=@n1.RGB, B=@n18), False=Multiply(A=@n1.RGB, B=@n4.G)) {parameterName=Edge}) {materialFunction=&MF_FX_EmissiveAdjustments}
  TextureCoordinate() {uTiling=9}
  Transform(None=VectorParameter() {parameterName=Position, defaultValue=<rgba 0 0 0 1>}.RGB) {transformSourceType=TRANSFORMSOURCE_Local}
 ]
}
