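# Unlit translucent Niagara ribbon material for the Kirby dual-trail vehicle drift cosmetic.
# A trail texture's red channel, tinted by particle color, feeds an emissive adjustment function.
# Opacity combines that channel with particle alpha and a square-root fade along U.
# A time-scrolled sine on the second UV set builds a local-Z wave offset, weighted along U and scaled by PowerWOP, which defaults to zero.
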Material /VehicleCosmetics/DriftTrail/DriftTrail_Kirby/Material/MM_DriftTrail_Kirby_DualTrail
graph {
 properties {
  bAutomaticallySetUsageInEditor=false
  bUsedWithNiagaraRibbons=true
  blendMode=BLEND_Translucent
  shadingModel=MSM_Unlit
  usedShadingModels=Unlit
 }
 nodes [
  n1 = TextureSample() {texture=/VehicleCosmetics/DriftTrail/DriftTrail_Kirby/Textures/T_VCVFX_DRiftrailKirby_TrailBase_RGB}
  MaterialExpressionVertexColor_0 = ParticleColor()
  MaterialFunctionCall(Emissive=Multiply(A=@n1.R, B=@MaterialExpressionVertexColor_0.RGB)) {materialFunction=&MF_FX_EmissiveAdjustments}
  n2 = ComponentMask(None=TextureCoordinate()) {r=true}
  Multiply(
   A=Multiply(A=Multiply(A=@n1.R, B=@MaterialExpressionVertexColor_0.A), B=Constant() {r=1})
   B=Power(Base=OneMinus(None=@n2), Exp=Constant() {r=0.5})
  )
  n3 = Constant()
  n4 = AppendVector(A=@n3, B=@n3)
  n5 = TextureCoordinate() {coordinateIndex=1}
  n6 = Time()
  n7 = Add(
   A=Divide(A=ComponentMask(None=@n5) {r=true}, B=ScalarParameter() {defaultValue=0.5, parameterName=CountWaves2})
   B=Multiply(A=ScalarParameter() {defaultValue=1, parameterName=Speed}, B=OneMinus(None=@n6))
  )
  n8 = ComponentMask(None=TextureCoordinate()) {r=true}
  n9 = ScalarParameter() {defaultValue=1, parameterName=PowerMiddle2}
  n10 = Multiply(
   A=Multiply(A=Sine(None=@n7), B=ScalarParameter() {defaultValue=7, parameterName=Strenght})
   B=Multiply(A=Power(Base=@n8, Exp=@n9), B=ScalarParameter() {defaultValue=0.5, parameterName=MultPower2})
  )
  Transform(None=Multiply(A=ScalarParameter() {parameterName=PowerWOP}, B=AppendVector(A=@n4, B=@n10))) {transformSourceType=TRANSFORMSOURCE_Local}
 ]
}
