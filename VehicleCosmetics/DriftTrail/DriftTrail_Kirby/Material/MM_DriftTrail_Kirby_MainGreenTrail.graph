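# Masked unlit Niagara material for the main green trail of the Kirby drift trail cosmetic, shared by ribbon and mesh particles.
# The Wheel0/Ribbon1 scalar blends between center-scaled wheel and plain ribbon UVs, and between simple alpha and a panning noise-and-foam mask fading along U.
# Color multiplies two scrolling dot layers, tiled by dynamic TileX on both axes, with particle color and a brightened base texture, through MF_FX_EmissiveAdjustments.
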Material /VehicleCosmetics/DriftTrail/DriftTrail_Kirby/Material/MM_DriftTrail_Kirby_MainGreenTrail
graph {
 properties {
  bAutomaticallySetUsageInEditor=false
  bUsedWithNiagaraMeshParticles=true
  bUsedWithNiagaraRibbons=true
  blendMode=BLEND_Masked
  shadingModel=MSM_Unlit
  usedShadingModels=Unlit
 }
 nodes [
  n1 = TextureCoordinate()
  n2 = AppendVector(A=Constant() {r=1}, B=ScalarParameter() {parameterName=ScaleUvs_V})
  n3 = Panner(Coordinate=LinearInterpolate(
    A=MaterialFunctionCall(UVs=@n1, Texture Scale=@n2) {materialFunction=&ScaleUVsByCenter}.UVs
    B=@n1
    Alpha=ScalarParameter() {defaultValue=1, parameterName=Wheel0/Ribbon1}
   ))
  n4 = TextureSample!(UVs=@n3, Tex, Level, Apply View MipBias) {
   mipValueMode=TMVM_MipLevel
   samplerSource=SSM_Clamp_WorldGroupSettings
   texture=/VehicleCosmetics/DriftTrail/DriftTrail_Kirby/Textures/T_VCVFX_BaseColor_D
  }
  MaterialExpressionVertexColor_0 = ParticleColor()
  n5 = Multiply(
   A=TextureCoordinate()
   B=AppendVector(A=ScalarParameter() {defaultValue=1, parameterName=X}, B=ScalarParameter() {defaultValue=1, parameterName=Y})
  )
  n6 = TextureSample!(UVs=Panner(Coordinate=TextureCoordinate()) {speedY=1.2}, Tex, Apply View MipBias) {texture=/VehicleCosmetics/DriftTrail/DriftTrail_Kirby/Textures/T_VCVFX_GradientKirbyFoam_A}
  n7 = Multiply(
   A=TextureSample!(UVs=Panner(Coordinate=@n5) {speedX=-1, speedY=0.1}, Tex, Level, Apply View MipBias) {mipValueMode=TMVM_MipLevel, texture=/VehicleCosmetics/DriftTrail/DriftTrail_Kirby/Textures/T_VCVFX_FluidNoise01_A}.R
   B=@n6.R
  )
  n8 = Constant() {r=100}
  n9 = Constant() {r=10}
  n10 = Multiply(A=@MaterialExpressionVertexColor_0.A, B=Add(A=Multiply(A=@n7, B=@n8), B=Power(Base=@n6.R, Exp=@n9)))
  n11 = OneMinus(None=ComponentMask(None=TextureCoordinate()) {r=true})
  LinearInterpolate(
   A=Multiply(A=@n4.A, B=@MaterialExpressionVertexColor_0.A)
   B=Multiply(A=@n4.A, B=Multiply(A=@n10, B=@n11))
   Alpha=ScalarParameter() {defaultValue=1, parameterName=Wheel0/Ribbon1}
  ) {constAlpha=0}
  n12 = TextureCoordinate() {uTiling=8}
  n13 = DynamicParameter() {paramNames [TileX,TileY,Param3,Param4]}
  n14 = AppendVector(A=@n13.TileX, B=@n13.TileX)
  n15 = Time()
  n16 = ScalarParameter() {parameterName=SpeedD}
  n17 = TextureSample!(UVs=Panner(Coordinate=Multiply(A=@n12, B=@n14), Time=Multiply(A=@n15, B=@n16)) {speedX=-3.3, speedY=0.1}, Tex, Bias, Apply View MipBias) {mipValueMode=TMVM_MipBias, texture=/VehicleCosmetics/DriftTrail/DriftTrail_Kirby/Textures/T_VCVFX_Dots2_A}
  n18 = TextureCoordinate() {uTiling=8}
  n19 = DynamicParameter() {paramNames [TileX,TileY,Param3,Param4]}
  n20 = AppendVector(A=@n19.TileX, B=@n19.TileX)
  n21 = Time()
  n22 = ScalarParameter() {parameterName=SpeedD2}
  n23 = TextureSample(UVs=Panner(Coordinate=Multiply(A=@n18, B=@n20), Time=Multiply(A=@n21, B=@n22)) {speedX=-3, speedY=-0.2}) {mipValueMode=TMVM_MipBias, texture=/VehicleCosmetics/DriftTrail/DriftTrail_Kirby/Textures/T_VCVFX_Dots_A}
  n24 = Add(
   A=TextureSample!(UVs=@n3, Tex, Apply View MipBias) {texture=&T_VCVFX_BaseColor_D}.R
   B=@n4.RGB
  )
  n25 = Multiply(A=@n4.R) {constB=2}
  n26 = Add(
   A=Multiply(A=@n24, B=ScalarParameter() {defaultValue=1, parameterName=MultiplierWhiterNucleus})
   B=Power(Base=Subtract(A=@n24, B=@n25), Exp=ScalarParameter() {defaultValue=1, parameterName=MultioplierColor}) {constExponent=1}
  )
  MaterialFunctionCall(Emissive=Multiply(A=Multiply(A=@n17.RGB, B=@n23.RGB), B=Multiply(A=@MaterialExpressionVertexColor_0.RGB, B=@n26))) {materialFunction=&MF_FX_EmissiveAdjustments}
  Transform(None=Multiply(A=WorldPosition().XYZ) {constB=0}) {transformSourceType=TRANSFORMSOURCE_Local}
 ]
}
